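; BP_AtomVisualizer is an Atom Framework inspection actor that displays Atom models in a grid or one model across its LODs.
; Construction script and UpdateInfo rebuild spawned components, centering each model's largest mesh and attaching its other parts.
; Text renders report materials, triangles and radius, recoloring over-budget models, alongside a set-wide summary.
; Custom events switch views, cycle models, rotate, change forced LOD and toggle Nanite via console commands.
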(event EventBeginPlay
  ;[1 0.6709 0.08789] Update Grid based on Game State
  (CallFunction|UpdateInfo))

(event Custom|UpdateInfo
  ;[1 0.6709 0.08789] Update Grid based on Game State
  (if (Variables|Data|GetGridView)
    (Setup|SetupComponentGridView)
    (else
      (Setup|SetupComponentLodView))))

(event Custom|Enable%20Nanite
  ;[0 0.846 1] Nanite and LOD Settings
  (Development|ExecuteConsoleCommand "r.Nanite 1")
  ;;[1 1 1] ForceLODModel 0 = Auto LOD
  (Variables|Data|SetLODLevel 0)
  (CallFunction|UpdateInfo))

(event Custom|LOD%20Bias%20Next
  (bind _lodlevel (Variables|Data|GetLODLevel))
  ;[0 0.846 1] Nanite and LOD Settings
  (Development|ExecuteConsoleCommand "r.Nanite 0")
  (if (< _lodlevel (Variables|NotEditable|GetNumLODs))
    (Math|Integer|IncrementInt _lodlevel
      (:%20%20
        _)))
  (CallFunction|UpdateInfo))

(event Custom|LOD%20Bias%20Previous
  (bind _lodlevel (Variables|Data|GetLODLevel))
  ;[0 0.846 1] Nanite and LOD Settings
  (Development|ExecuteConsoleCommand "r.Nanite 0")
  (if (> _lodlevel 0)
    (Math|Integer|DecrementInt _lodlevel
      (:%20%20
        _))
    (else
      _))
  (CallFunction|UpdateInfo))

(event Custom|Disable%20Nanite/LOD%20Auto
  ;[0 0.846 1] Nanite and LOD Settings
  (Development|ExecuteConsoleCommand "r.Nanite 0")
  ;;[1 1 1] ForceLODModel 0 = Auto LOD
  (Variables|Data|SetLODLevel 0)
  (CallFunction|UpdateInfo))

(event Custom|-%20GRID%20ViewMode%20-%20
  ;[0 1 0.1981] View Modes
  (Variables|Data|SetGridView true)
  (CallFunction|UpdateInfo))

(event Custom|-%20LOD%20ViewMode%20-
  ;[0 1 0.1981] View Modes
  (Variables|Data|SetGridView false)
  (CallFunction|UpdateInfo))

(event Custom|ModelNext
  (bind _currentindex (Variables|NotEditable|GetCurrentIndex))
  ;[1 0.3843 0.1248] Cycle Prefabs
  (Math|Integer|IncrementInt (Variables|NotEditable|GetCurrentIndex)
    (:%20%20
      (if (> _currentindex (Variables|NotEditable|GetLastModelIndex))
        (Variables|NotEditable|SetCurrentIndex 0)
        (elif (< _currentindex 0)
          (Variables|NotEditable|SetCurrentIndex (Variables|NotEditable|GetLastModelIndex))
          (else
            _)))
      (if (Variables|Data|GetGridView)
        (Variables|Data|SetAtomModelToDisplay (Utilities|Array|Get(aref) (Variables|Data|GetFilteredMeshIDs) (Variables|NotEditable|GetCurrentIndex)))
        (else
          (Variables|Data|SetAtomModelToDisplay (Utilities|Array|Get(aref) (Variables|Data|GetFilteredMeshIDs) (Variables|NotEditable|GetCurrentIndex)))
          (CallFunction|UpdateInfo))))))

(event Custom|ModelPrevious
  (bind _currentindex (Variables|NotEditable|GetCurrentIndex))
  ;[1 0.3843 0.1248] Cycle Prefabs
  (Math|Integer|DecrementInt (Variables|NotEditable|GetCurrentIndex)
    (:%20%20
      (if (> _currentindex (Variables|NotEditable|GetLastModelIndex))
        (Variables|NotEditable|SetCurrentIndex 0)
        (elif (< _currentindex 0)
          (Variables|NotEditable|SetCurrentIndex (Variables|NotEditable|GetLastModelIndex))
          (else
            _)))
      (if (Variables|Data|GetGridView)
        (Variables|Data|SetAtomModelToDisplay (Utilities|Array|Get(aref) (Variables|Data|GetFilteredMeshIDs) (Variables|NotEditable|GetCurrentIndex)))
        (else
          (Variables|Data|SetAtomModelToDisplay (Utilities|Array|Get(aref) (Variables|Data|GetFilteredMeshIDs) (Variables|NotEditable|GetCurrentIndex)))
          (CallFunction|UpdateInfo))))))

(event Custom|RotateClock
  ;[1 1 1] Rotation Control
  (Variables|Setup|SetYawCustomRotation (+ (Variables|Setup|GetYawCustomRotation) (Variables|Data|GetRotationIncrement)))
  (CallFunction|UpdateInfo))

(event Custom|RotateAntiClock
  ;[1 1 1] Rotation Control
  (Variables|Setup|SetYawCustomRotation (- (Variables|Setup|GetYawCustomRotation) (Variables|Data|GetRotationIncrement)))
  (CallFunction|UpdateInfo))

(event Custom|RotationReset
  ;[1 1 1] Rotation Control
  (Variables|Setup|SetYawCustomRotation 0.0)
  (CallFunction|UpdateInfo))

(fn AddComponentInfoDisplay
  (params (LOD%20Index :type Integer) (AssetIndex :type Integer) (Location :type Vector) (LocalOffset :type Vector) (AdditionalString :type String) (OverrideTextColor :type Boolean) (OverrideColor :type Linear%20Color%20Structure) (IsAutoLod :type Boolean) (IsLodView :type Boolean))
  (bind _returnvalue (AddComponent|NODE_AddTextRenderComponent-4 :bManualAttachment true :RelativeTransform_Location Location :RelativeTransform_Rotation "90.000000,0.000000,0.000000" :RelativeTransform_Scale "1.000000,1.000000,1.000000"))
  (if OverrideTextColor
    (Rendering|Components|TextRender|SetTextRenderColor _returnvalue (Math|Conversions|ToColor(LinearColor) OverrideColor))
    (else
      (Rendering|Components|TextRender|SetTextRenderColor _returnvalue (Variables|Display|GetTextColor))))
  (Transformation|AttachComponentToComponent _returnvalue 0 "None" "KeepWorld" "KeepWorld" "KeepWorld" false)
  (Rendering|Components|TextRender|SetWorldSize _returnvalue (Variables|Display|GetFontSize))
  (Rendering|Components|TextRender|SetHorizontalAlignment _returnvalue "EHTA_Center")
  (Utilities|Array|Add (Variables|NotEditable|GetChildComponents) _returnvalue)
  (DisplayInfo|SetComponentDisplayInfo _returnvalue LOD%20Index AssetIndex AdditionalString OverrideTextColor (Variables|GetIsAutoLOD) (Variables|GetIsLodView))
  (Transformation|AddLocalOffset _returnvalue LocalOffset false false)
  (return _returnvalue))

(fn AttachAdditionalParts
  (params (AdditionalMeshes :type Array%20of%20Static%20Mesh%20Soft%20Object%20References) (AtomStaticMeshComponent :type Static%20Mesh%20Component%20Object%20Reference) (LODModel :type Integer))
  (if (Utilities|Array|IsNotEmpty AdditionalMeshes)
    (for _array_element AdditionalMeshes
      (Utilities|IsValid (Utilities|ResolveSoftReference _array_element)
        (:Is%20Valid
          (bind _returnvalue (AddComponent|NODE_AddStaticMeshComponent-6 :bManualAttachment true))
          (Transformation|AttachComponentToComponent _returnvalue AtomStaticMeshComponent "None" "KeepRelative" "KeepRelative" "KeepRelative" true)
          (Components|StaticMesh|SetStaticMesh _returnvalue (Utilities|ResolveSoftReference _array_element))
          (if (not (Variables|Data|GetGridView))
            (Rendering|LOD|SetForcedLodModel _returnvalue LODModel))
          (Utilities|Array|Add (Variables|NotEditable|GetChildComponents) _returnvalue))
        (:Is%20Not%20Valid
          (return))))
    (else
      _)))

(fn CenterSceneComponent
  (params (StaticMeshComponent :type Static%20Mesh%20Component%20Object%20Reference))
  (bind (_origin _boxextent) (Utilities|GetLocalBounds StaticMeshComponent))
  (bind (_returnvalue_x _returnvalue_y _returnvalue_z) (Math|Vector|BreakVector (- (Variables|/Script/Engine.SceneComponent|GetRelativeLocation StaticMeshComponent) _origin)))
  (Transformation|SetRelativeLocation StaticMeshComponent (Math|Vector|MakeVector _returnvalue_x _returnvalue_y (+ _returnvalue_z (.z _boxextent))) false false)
  (return (Transformation|GetRelativeTransform StaticMeshComponent)))

(fn ComposeAssetInfo
  (params (TextRenderComp :type Text%20Render%20Component%20Object%20Reference) (AdditionalString :type String) (AssetIndex :type Integer) (AssetName :type String) (NumMaterials :type Integer) (LODIndex :type Integer) (AssetTriCount :type Integer) (SphereRadius :type Integer) (IsAutoLOD :type Boolean))
  (Rendering|Components|TextRender|SetText TextRenderComp (Utilities|Text|ToText(String) (Utilities|String|Append (Utilities|String|Append AdditionalString (Utilities|String|ToString(Integer) AssetIndex) " | ") (Utilities|String|Append AssetName "
") (Utilities|String|Append "Number of Materials: " (Utilities|String|ToString(Integer) NumMaterials) "
") (select (Variables|GetIsAutoLOD) (Utilities|String|Append "LOD " "Auto" " | ") (Utilities|String|Append "LOD" (Utilities|String|ToString(Integer) LODIndex) " | ")) (Utilities|String|Append "Tri Count: " (Utilities|String|ToString(Integer) AssetTriCount) "
Sphere Radius: " (Utilities|String|ToString(Integer) SphereRadius) "
") "")))
  (return))

(fn GetHighestLOD
  (params (AtomMeshes :type Array%20of%20STRUCT%20Atom%20Visualizer%20Structures) (IndexToDisplay :type Integer))
  (bind (_atommodel_35_6de1c46d446db74967e7e7aad1f0a2c5 _atommeshes_39_3194e7ed4270b50117be52a4b3c73dd0) (Utilities|Struct|BreakSTRUCTAtomVisualizer (Utilities|Array|Get(aref) AtomMeshes IndexToDisplay)))
  (for _array_element _atommeshes_39_3194e7ed4270b50117be52a4b3c73dd0
    (bind _returnvalue (Utilities|LoadAssetBlocking _array_element))
    (bind _asstatic_mesh (Utilities|Casting|CastToStaticMesh _returnvalue))
    (if (> (StaticMesh|GetNumLODs _asstatic_mesh) (LocalVariables|GetTempLOD))
      (LocalVariables|SetTempLOD (StaticMesh|GetNumLODs _asstatic_mesh))
      (LocalVariables|SetTempStaticMesh _asstatic_mesh)))
  (return (LocalVariables|GetTempLOD) (LocalVariables|GetTempStaticMesh)))

(fn GetLargestMesh
  (params (StaticMeshes :type Array%20of%20Static%20Mesh%20Soft%20Object%20References))
  (for _array_element StaticMeshes
    (bind _returnvalue (Utilities|LoadAssetBlocking _array_element))
    (bind _asstatic_mesh (Utilities|Casting|CastToStaticMesh _returnvalue))
    (bind (_min _max) (Utilities|Struct|BreakBox (StaticMesh|GetBoundingBox _asstatic_mesh)))
    (bind (_min_x _min_y) (Math|Vector|BreakVector _min))
    (bind (_max_x _max_y) (Math|Vector|BreakVector _max))
    (if (> (Math|Float|Sqrt (+ (* (- _max_x _min_x) (- _max_x _min_x)) (* (- _max_y _min_y) (- _max_y _min_y)))) (LocalVariables|GetTempBaseSize))
      (LocalVariables|SetTempBaseSize (Math|Float|Sqrt (+ (* (- _max_x _min_x) (- _max_x _min_x)) (* (- _max_y _min_y) (- _max_y _min_y)))))
      (LocalVariables|SetTempStaticMesh _asstatic_mesh)))
  (return (LocalVariables|GetTempBaseSize) (LocalVariables|GetTempStaticMesh)))

(fn GetLocalBounds
  (params (StaticMeshComponent :type Scene%20Component%20Object%20Reference))
  (bind (_origin _boxextent _sphereradius) (Collision|GetComponentBounds StaticMeshComponent))
  (return (Math|Transform|InverseTransformLocation (Transformation|GetWorldTransform StaticMeshComponent) _origin) _boxextent _sphereradius))

(fn InitializeCommon
  (bind _filteredmeshids (Variables|Data|GetFilteredMeshIDs))
  (bind _returnvalue (Utilities|Array|Length _filteredmeshids))
  (bind _atommodeltodisplay (Variables|Data|GetAtomModelToDisplay))
  (bind _filteredmeshids_2 (Variables|Data|GetFilteredMeshIDs))
  (Rendering|Components|TextRender|SetText (Variables|Default|GetInfoDisplay) (Utilities|Text|MakeLiteralText ""))
  (if (and (Utilities|Array|IsNotEmpty (Variables|Data|GetAtomMeshes)) (Utilities|Array|IsNotEmpty (Variables|Data|GetFilteredMeshIDs)))
    (Utilities|FlowControl|Sequence
      (:then_0
        (if (Utilities|Array|ContainsItem _filteredmeshids_2 _atommodeltodisplay)
          (Variables|NotEditable|SetCurrentIndex (Utilities|Array|FindItem _filteredmeshids_2 _atommodeltodisplay))
          (else
            (Variables|NotEditable|SetCurrentIndex 0)))
        (Variables|Data|SetAtomModelToDisplay (Utilities|Array|Get(aref) _filteredmeshids_2 (Variables|NotEditable|GetCurrentIndex))))
      (:then_1
        (Variables|NotEditable|SetNumModels _returnvalue)
        (Variables|NotEditable|SetLastModelIndex (- _returnvalue 1))
        (Variables|Data|SetAtomModelToDisplay (Utilities|Array|Get(aref) (Variables|Data|GetFilteredMeshIDs) (Variables|NotEditable|GetCurrentIndex)))
        (bind _highestlod (Utilities|GetHighestLOD (Variables|Data|GetAtomMeshes) (Variables|Data|GetAtomModelToDisplay)))
        (Variables|NotEditable|SetNumLODs _highestlod)))
    (else
      (Variables|NotEditable|SetNumModels 0)))
  (return))

(fn InitializeValues_GridView
  (bind _childcomponents (Variables|NotEditable|GetChildComponents))
  (Initialization|InitializeCommon)
  (Utilities|FlowControl|Sequence
    (:then_0
      ;[1 1 1] Clears any existing component
      (Variables|NotEditable|SetControlID 0)
      (if (Utilities|Array|IsNotEmpty _childcomponents)
        (for _array_element _childcomponents
          (Components|DestroyComponent _array_element))
        (Utilities|Array|Clear (Variables|NotEditable|GetChildComponents))))
    (:then_1
      ;[1 1 1] Defines the Location on which the first component of the Grid will be spawned
      (bind _output_get (Variables|NotEditable|SetInitialPosX 0.0))
      (Variables|NotEditable|SetCurrentPosX _output_get))
    (:then_2
      (bind _output_get_2 (Variables|NotEditable|SetInitialPosY 0.0))
      (Variables|NotEditable|SetCurrentPosY _output_get_2))
    (:then_3
      (bind _output_get_3 (Variables|NotEditable|SetInitialPosZ (Variables|Data|GetHeightOffset)))
      (Variables|NotEditable|SetCurrentPosZ _output_get_3))
    (:then_4
      ;[1 1 1] Check the biggest Radius and set it as Base Size
      (Variables|NotEditable|SetBaseSize (+ (+ (Math|Float|Max(Float) (Variables|Data|GetLargestBound) 1.7e+02) 3e+01) (Variables|Setup|GetAdditionalSpacingOnGrid)))
      (return))))

(fn InitializeValues_LodView
  (bind _childcomponents (Variables|NotEditable|GetChildComponents))
  (Initialization|InitializeCommon)
  ;[1 1 1] Clears any existing component
  (Variables|NotEditable|SetCurrentOffsetSingle 0.0)
  (if (Utilities|Array|IsNotEmpty _childcomponents)
    (for _array_element _childcomponents
      (Components|DestroyComponent _array_element))
    (Utilities|Array|Clear (Variables|NotEditable|GetChildComponents))
    (else
      _))
  (return))

(fn SetComponentDisplayInfo
  (params (TextRenderComponent :type Text%20Render%20Component%20Object%20Reference) (LOD%20Index :type Integer) (AssetIndex :type Integer) (AdditionalString :type String) (OverrideColor :type Boolean) (IsAutoLod :type Boolean) (IsLodView :type Boolean))
  (bind _atommeshes (Variables|Data|GetAtomMeshes))
  (bind _output (Utilities|Array|Get(aref) _atommeshes AssetIndex))
  (bind (_atommodel_35_6de1c46d446db74967e7e7aad1f0a2c5 _atommeshes_39_3194e7ed4270b50117be52a4b3c73dd0) (Utilities|Struct|BreakSTRUCTAtomVisualizer _output))
  (bind (_origin _boxextent _sphereradius) (Math|BoxSphereBounds|BreakBoxSphereBounds (StaticMesh|GetBounds (LocalVariables|GetTempMesh))))
  (if (Variables|GetIsLodView)
    (LocalVariables|SetLODLevelDisplay (Variables|GetLOD%20Index))
    (else
      (LocalVariables|SetLODLevelDisplay (Math|Integer|Max(Integer) (- LOD%20Index 1)))))
  (Utilities|FlowControl|Sequence
    (:then_0
      (bind _returnvalue (Utilities|LoadAssetBlocking _atommodel_35_6de1c46d446db74967e7e7aad1f0a2c5))
      (bind _asatom_model (Utilities|Casting|CastToAtomModel _returnvalue)
        (:then
          (for _array_element _atommeshes_39_3194e7ed4270b50117be52a4b3c73dd0
            (bind _returnvalue_2 (Utilities|LoadAssetBlocking _array_element))
            (bind _asstatic_mesh (Utilities|Casting|CastToStaticMesh _returnvalue_2))
            (Utilities|IsValid _asstatic_mesh
              (:Is%20Valid
                (LocalVariables|SetMeshTriCount (+ (LocalVariables|GetMeshTriCount) (StaticMesh|GetNumTriangles _asstatic_mesh (LocalVariables|GetLODLevelDisplay))))))))
        (:CastFailed
          (Rendering|Components|TextRender|SetText TextRenderComponent (Utilities|Text|ToText(String) "Empty Selection")))))
    (:then_1
      (DisplayInfo|ComposeAssetInfo TextRenderComponent AdditionalString AssetIndex (select (.ReturnValue (Utilities|String|Split (Utilities|GetDisplayName _asatom_model) "SM_")) (.RightS (Utilities|String|Split (Utilities|GetDisplayName _asatom_model) "SM_")) (Utilities|GetDisplayName _asatom_model)) (StaticMesh|GetNumSections (LocalVariables|GetTempMesh) (LocalVariables|GetLODLevelDisplay)) (LocalVariables|GetLODLevelDisplay) (LocalVariables|GetMeshTriCount) (Math|Float|Round _sphereradius) (Variables|GetIsAutoLOD))
      (if (and (> (LocalVariables|GetMeshTriCount) (Variables|Data|GetBudget)) (>= (LocalVariables|GetLODLevelDisplay) (Variables|Data|GetBudgetedLOD)))
        (if (not (Variables|GetOverrideColor))
          (Rendering|Components|TextRender|SetTextRenderColor TextRenderComponent (Variables|NotEditable|GetOverBudgetedColor)))
        (else
          _))))
  (return))

(fn SetGeneralDisplayInfo
  (params (DisplayPositionOffset :type Float%20%28double-precision%29))
  (bind (_x _y _z) (Utilities|Struct|BreakIntVector (Variables|Data|GetRadiusInfo)))
  (bind (_origin _boxextent) (Collision|GetComponentBounds (Variables|Default|GetInfoDisplay)))
  (LocalVariables|SetIsAutoLOD (== (Variables|Data|GetLODLevel) 0))
  (LocalVariables|SetLODLevelDisplay (Math|Integer|Max(Integer) (- (Variables|Data|GetLODLevel) 1)))
  (Rendering|Components|TextRender|SetTextRenderColor (Variables|Default|GetInfoDisplay) (Variables|Display|GetTextColor))
  (Rendering|Components|TextRender|SetWorldSize (Variables|Default|GetInfoDisplay) (Variables|Display|GetGeneralDisplayFontSize))
  (Rendering|Components|TextRender|SetText (Variables|Default|GetInfoDisplay) (Utilities|Text|ToText(String) (Utilities|String|Append (Utilities|String|Append (Variables|Data|GetLoadedPath) "
Number of valid Assets: " (Utilities|String|ToString(Integer) (Variables|NotEditable|GetNumModels))) (Utilities|String|Append "
Displaying LOD" (select (LocalVariables|GetIsAutoLOD) " Auto" (Utilities|String|ToString(Integer) (LocalVariables|GetLODLevelDisplay)))) (Utilities|String|Append "
Average Triangle Count: " (Utilities|String|ToString(Integer) (Variables|Data|GetAverageTriangleCount))) (Utilities|String|Append "
SphereRadius: 
    Average:  " (Utilities|String|ToString(Integer) _x) "
    Largest:  " (Utilities|String|ToString(Integer) _y) "
    Smalllest:  " (Utilities|String|ToString(Integer) _z)) (Utilities|StringCombine (Utilities|String|Append "

Budget(LOD" (Utilities|String|ToString(Integer) (Variables|Data|GetBudgetedLOD)) "): " (Utilities|String|ToString(Integer) (Variables|Data|GetBudget)) "
Over-Budgeted Models: " (Utilities|String|ToString(Integer) (Utilities|Array|Length (Variables|Data|GetOverBudgetedIDs))) "

") (Variables|Data|GetOverBudgetedIDs) (Variables|NotEditable|GetHighestTriangleNames) (Variables|NotEditable|GetHighestTriangleCounts)))))
  (LocalVariables|SetTempSceneComponents (Components|GetChildrenComponents (Variables|Default|GetScene) true))
  (Transformation|SetRelativeLocation (Variables|Default|GetInfoDisplay) (Math|Vector|MakeVector 0.0 (+ (* (.y _boxextent) 2.0) DisplayPositionOffset)) false false)
  (return))

(fn SetupComponent_GridView
  (bind _controlid (Variables|NotEditable|GetControlID))
  (bind _filteredmeshids (Variables|Data|GetFilteredMeshIDs))
  (bind _output (Utilities|Array|Get(aref) _filteredmeshids _controlid))
  (bind _atommeshes (Variables|Data|GetAtomMeshes))
  (bind _output_2 (Utilities|Array|Get(aref) _atommeshes _output))
  (bind (_atommodel_35_6de1c46d446db74967e7e7aad1f0a2c5 _atommeshes_39_3194e7ed4270b50117be52a4b3c73dd0) (Utilities|Struct|BreakSTRUCTAtomVisualizer _output_2))
  (Initialization|InitializeValuesGridView)
  (if (and (Utilities|Array|IsNotEmpty (Variables|Data|GetAtomMeshes)) (Utilities|Array|IsNotEmpty (Variables|Data|GetFilteredMeshIDs)))
    (bind _transform (Utilities|GridLoop (Variables|NotEditable|GetNumModels))
      (:LoopBody
        (if (<= (Variables|NotEditable|GetControlID) (Variables|NotEditable|GetLastModelIndex))
          (bind _returnvalue (AddComponent|NODE_AddStaticMeshComponent-9 :RelativeTransform _transform)) ; Dummy Parent
          (Transformation|SetRelativeRotation _returnvalue (Math|Rotator|MakeRotator 0.0 0.0 (Variables|Setup|GetYawCustomRotation)) false false)
          (bind _returnvalue_2 (AddComponent|NODE_AddStaticMeshComponent-2 :bManualAttachment true))
          (Transformation|AttachComponentToComponent _returnvalue_2 _returnvalue "None" "KeepRelative" "KeepRelative" "KeepRelative" true)
          (LocalVariables|SetTempAtomMeshes _atommeshes_39_3194e7ed4270b50117be52a4b3c73dd0)
          (bind (_largestbound _largestboundmesh) (Utilities|GetLargestMesh _atommeshes_39_3194e7ed4270b50117be52a4b3c73dd0))
          (Utilities|IsValid _largestboundmesh
            (:Is%20Valid
              (Utilities|Array|RemoveItem (LocalVariables|GetTempAtomMeshes) (Utilities|ResolveSoftReference _largestboundmesh))
              ;[1 1 1] Tag Used to find components to highlight as expensive meshes
              (Variables|/Script/Engine.ActorComponent|SetComponentTags (Utilities|Array|MakeArray (Utilities|String|StringToName (Utilities|GetDisplayName _largestboundmesh))) _returnvalue_2)
              (Components|StaticMesh|SetStaticMesh _returnvalue_2 _largestboundmesh)
              (Utilities|Array|Add (Variables|NotEditable|GetChildComponents) _returnvalue_2)
              (Rendering|LOD|SetForcedLodModel _returnvalue_2 (Variables|Data|GetLODLevel))
              (Utilities|CenterSceneComponent _returnvalue_2)
              (Setup|AttachAdditionalParts (LocalVariables|GetTempAtomMeshes) _returnvalue_2 (Variables|Data|GetLODLevel))
              ;[1 1 1] Adjusts the position of the Text Render Dynamically based on the selected mesh
              (bind (_origin _boxextent) (Collision|GetComponentBounds _returnvalue_2))
              (bind (_origin_x _origin_y) (Math|Vector|BreakVector _origin))
              (bind _boxextent_x (Math|Vector|BreakVector _boxextent))
              (bind _returnvalue_3 (AddComponent|NODE_AddTextRenderComponent-1 :bManualAttachment true :RelativeTransform_Location_X _origin_x :RelativeTransform_Location_Y _origin_y :RelativeTransform_Rotation "90.000000,0.000000,0.000000" :RelativeTransform_Scale "1.000000,1.000000,1.000000"))
              (Transformation|AttachComponentToComponent _returnvalue_3 _returnvalue_2 "None" "KeepWorld" "KeepWorld" "KeepWorld" false)
              (Rendering|Components|TextRender|SetHorizontalAlignment _returnvalue_3 "EHTA_Center")
              (Rendering|Components|TextRender|SetTextRenderColor _returnvalue_3 (Variables|Display|GetTextColor))
              (Transformation|AddLocalOffset _returnvalue_3 (Math|Vector|MakeVector 0.0 0.0 (* (+ _boxextent_x 1e+01) -1.0)) false false)
              (Utilities|Array|Add (Variables|NotEditable|GetChildComponents) _returnvalue_3)
              (DisplayInfo|SetComponentDisplayInfo _returnvalue_3 (Variables|Data|GetLODLevel) (Utilities|Array|Get(aref) (Variables|Data|GetFilteredMeshIDs) (Variables|NotEditable|GetControlID)) "" false (== (Variables|Data|GetLODLevel) 0))
              (Variables|NotEditable|SetCurrentPosY (- (Variables|NotEditable|GetCurrentPosY) (Variables|NotEditable|GetBaseSize)))
              (Math|Integer|IncrementInt (Variables|NotEditable|GetControlID))))
          (else
            _))))
    (else
      _))
  (DisplayInfo|SetGeneralDisplayInfo (* (Variables|NotEditable|GetBaseSize) 0.5))
  (return))

(fn SetupComponent_LodView
  (bind _atommodeltodisplay (Variables|Data|GetAtomModelToDisplay))
  (bind _atommeshes (Variables|Data|GetAtomMeshes))
  (bind _output (Utilities|Array|Get(aref) _atommeshes _atommodeltodisplay))
  (bind (_atommodel_35_6de1c46d446db74967e7e7aad1f0a2c5 _atommeshes_39_3194e7ed4270b50117be52a4b3c73dd0) (Utilities|Struct|BreakSTRUCTAtomVisualizer _output))
  (Initialization|InitializeValuesLodView)
  (if (and (Utilities|Array|IsNotEmpty (Variables|Data|GetAtomMeshes)) (Utilities|Array|IsNotEmpty (Variables|Data|GetFilteredMeshIDs)))
    (for _index (range (+ (Variables|NotEditable|GetNumLODs) 1))
      (bind _returnvalue (AddComponent|NODE_AddStaticMeshComponent-1 :RelativeTransform_Location_Y (Variables|NotEditable|GetCurrentOffsetSingle) :RelativeTransform_Location_Z (Variables|Data|GetHeightOffset) :RelativeTransform_Rotation_Yaw (Variables|Setup|GetYawCustomRotation) :RelativeTransform_Scale "1.000000,1.000000,1.000000"))
      (LocalVariables|SetTempAtomMeshes _atommeshes_39_3194e7ed4270b50117be52a4b3c73dd0)
      (bind (_largestbound _largestboundmesh) (Utilities|GetLargestMesh _atommeshes_39_3194e7ed4270b50117be52a4b3c73dd0))
      (Utilities|Array|RemoveItem (LocalVariables|GetTempAtomMeshes) (Utilities|ResolveSoftReference _largestboundmesh))
      (Components|StaticMesh|SetStaticMesh _returnvalue _largestboundmesh)
      (Utilities|CenterSceneComponent _returnvalue)
      (Rendering|LOD|SetForcedLodModel _returnvalue _index)
      ;[1 1 1] Tag Used to find components to highlight as expensive meshes
      (Variables|/Script/Engine.ActorComponent|SetComponentTags (Utilities|Array|MakeArray (Utilities|String|StringToName (Utilities|GetDisplayName _largestboundmesh))) _returnvalue)
      (Utilities|Array|Add (Variables|NotEditable|GetChildComponents) _returnvalue)
      (Setup|AttachAdditionalParts (LocalVariables|GetTempAtomMeshes) _returnvalue _index)
      (if (== _index 0)
        ;[1 0.3476 0] LOD Transition Preview
        (Rendering|LOD|SetForcedLodModel _returnvalue (Variables|Data|GetLODLevel))
        (bind (_origin _boxextent) (Collision|GetComponentBounds _returnvalue))
        (bind (_origin_x _origin_y) (Math|Vector|BreakVector _origin))
        (bind (_boxextent_x _boxextent_y) (Math|Vector|BreakVector _boxextent))
        (bind _textrendercomponent (DisplayInfo|AddComponentInfoDisplay (Variables|Data|GetLODLevel) (Variables|Data|GetAtomModelToDisplay) (Math|Vector|MakeVector _origin_x _origin_y) (Math|Vector|MakeVector 0.0 0.0 (* (+ _boxextent_x 1e+01) -1.0)) (Utilities|String|Append "LOD TRANSITION PREVIEW
" "
") true "(R=1.000000,G=1.000000,B=1.000000,A=1.000000)" (== (Variables|Data|GetLODLevel) 0)))
        (bind (_origin_2 _boxextent_2) (Collision|GetComponentBounds _textrendercomponent))
        (bind (_boxextent_x_2 _boxextent_y_2) (Math|Vector|BreakVector _boxextent_2))
        (Variables|NotEditable|SetCurrentOffsetSingle (Utilities|Operators|Subtract (Variables|NotEditable|GetCurrentOffsetSingle) (Math|Float|Max(Float) (* 2.0 _boxextent_y) _boxextent_y_2) (Variables|Setup|GetAdditionalSpacingLODs) 2e+02))
        (LocalVariables|SetSpacingFromGeneralDisplay (Math|Float|Max(Float) _boxextent_y_2 _boxextent_y))
        (else
          (bind _textrendercomponent_2 (DisplayInfo|AddComponentInfoDisplay :LOD%20Index (- _index 1) :AssetIndex (Variables|Data|GetAtomModelToDisplay) :Location (Math|Vector|MakeVector _origin_x _origin_y) :LocalOffset (Math|Vector|MakeVector 0.0 0.0 (* (+ _boxextent_x 1e+01) -1.0)) :OverrideTextColor false :IsAutoLod false :IsLodView true))
          (Variables|NotEditable|SetCurrentOffsetSingle (Utilities|Operators|Subtract (Variables|NotEditable|GetCurrentOffsetSingle) (Math|Float|Max(Float) (* 2.0 _boxextent_y) (* (.SphereRadius (Collision|GetComponentBounds _textrendercomponent_2)) 2.0)) 1e+01 (Variables|Setup|GetAdditionalSpacingLODs))))))
    (else
      _))
  (DisplayInfo|SetGeneralDisplayInfo (* (LocalVariables|GetSpacingFromGeneralDisplay) 2.0))
  (return))

(fn StringCombine
  (params (Base%20String :type String) (IDs :type Array%20of%20Integers) (Names :type Array%20of%20Strings) (TriangleCounts :type Array%20of%20Integers))
  (LocalVariables|SetNewString Base%20String)
  (for (_array_element _array_index) Names
    (LocalVariables|SetNewString (Utilities|String|Append (LocalVariables|GetNewString) (Utilities|String|ToString(Integer) (Utilities|Array|Get(aref) IDs _array_index)) " | " _array_element " - " (Utilities|String|ToString(Integer) (Utilities|Array|Get(aref) TriangleCounts _array_index)) " tris" "
")))
  (return (LocalVariables|GetNewString)))

(fn ConstructionScript
  (Transformation|SetActorRotation self (Math|Rotator|MakeRotator 0.0 0.0 (Math|Conversions|ToFloat(Integer) (* (Math|Float|Round (/ (Math|Conversions|ToFloat(Integer) (Math|Float|Round (.yaw (Transformation|GetActorRotation self)))) 180)) 180))) false)
  (if (Variables|Setup|Advanced|GetRunConstructor)
    (if (Variables|Data|GetGridView)
      (Setup|SetupComponentGridView)
      (else
        (Setup|SetupComponentLodView)))))
