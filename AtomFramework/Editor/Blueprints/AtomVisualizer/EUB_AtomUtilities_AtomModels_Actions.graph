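# Editor utility blueprint in the AtomFramework editor tools that exposes a ProcessSelectedAtomModels action, apparently for selected Atom model assets.
# The function is a thin wrapper that forwards to a same-named ProcessSelectedAtomModels call, passing itself as world context.
# The event graph is empty and the blueprint holds no variables, so the actual processing lives in the called function.
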EditorUtilityBlueprint /AtomFramework/Editor/Blueprints/AtomVisualizer/EUB_AtomUtilities_AtomModels_Actions
graphs [
 EventGraph {links [], nodes []}
 ProcessSelectedAtomModels {nodes [
   ProcessSelectedAtomModels() {is=title K2Node_FunctionEntry}
   > ProcessSelectedAtomModels!(execute, __WorldContext:Object Reference=Self-Reference()) -> (then) {is=title K2Node_CallFunction}
  ]}
]
variables []
